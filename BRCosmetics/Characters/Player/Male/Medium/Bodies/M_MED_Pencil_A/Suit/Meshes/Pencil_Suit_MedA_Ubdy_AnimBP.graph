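# Animation Blueprint for the upper-body mesh of the medium male Pencil Suit cosmetic, following the pose of its attached parent mesh.
# The copied pose runs through a deform Control Rig, then an upper-body Control Rig whose athletic-lower switch is set each update from a character meta tag.
# A weighted multi-bone constraint then drives the badge bone from spine and left clavicle, and additive offsets nudge the cape and backpack attach bones.
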AnimBlueprint /BRCosmetics/Characters/Player/Male/Medium/Bodies/M_MED_Pencil_A/Suit/Meshes/Pencil_Suit_MedA_Ubdy_AnimBP
target_skeleton=&/Pencil_Suit_MedA_Ubdy_Skeleton
graphs [
 AnimGraph {nodes [
   Input Pose()
   n1 = Control Rig(Source=Copy Pose From Mesh() {bUseAttachedParent=true}) {
    alphaBoolBlend {blendOption=Linear}
    controlRigAssetReference {blueprintRigClass=&/Pencil_Suit_MedA_DeformRig.Pencil_Suit_MedA_DeformRig_C}
   }
   n2 = Control Rig!(Source=@n1, Alpha, AthleticLowerCtrl:bool=Get AthleticBool() -> (AthleticBool:bool)) -> (Pose) {
    alphaBoolBlend {blendOption=Linear}
    controlRigAssetReference {blueprintRigClass=&/Pencil_Suit_MedA_Ubdy_CtrlRig.Pencil_Suit_MedA_Ubdy_CtrlRig_C}
   }
   n3 = "Constraint - dyn_badge (PPPPP)"!(
    ComponentPose=Local To Component(LocalPose=@n2)
    Alpha=1
    ConstraintWeights_0=0.1246
    ConstraintWeights_1:float=0.2112
    ConstraintWeights_2:float=0.0451
    ConstraintWeights_3:float=0.0507
    ConstraintWeights_4:float=0.5684
   ) -> (Pose) {
    alphaBoolBlend {blendOption=Linear}
    boneToModify {boneName=dyn_badge}
    constraintSetup [
     {targetBone {boneName=spine_05}}
     {targetBone {boneName=spine_03}}
     {targetBone {boneName=spine_02}}
     {targetBone {boneName=clavicle_l}}
     {targetBone {boneName=spine_04}}
    ]
    constraintWeights [1,1,1,1,1]
   }
   n4 = "Transform (Modify) Bone - Bone: attach_cape"(ComponentPose=@n3, Alpha=1, Translation="0,1.5,-1", Rotation="0,0,0", Scale="1,1,1") {translationMode=BMM_Additive, alphaBoolBlend {blendOption=Linear}, boneToModify {boneName=attach_cape}, translation {y=1.5, z=-1}}
   n5 = "Transform (Modify) Bone - Bone: attach_backpack"(ComponentPose=@n4, Alpha=1, Translation="0,1.5,0", Rotation="0,0,0", Scale="1,1,1") {translationMode=BMM_Additive, alphaBoolBlend {blendOption=Linear}, boneToModify {boneName=attach_backpack}, translation {y=1.5}}
   Output Pose(Result=Component To Local(ComponentPose=@n5)) {layerGroup=None}
  ]}
 EventGraph {nodes [
   Event BlueprintUpdateAnimation()
   > Set AthleticBool(AthleticBool:bool=CharacterItemDefinitionHasAnyMetaTag(Tags=Make Array("[0]"=(TagName="Cosmetics.MetaTags.Pencil.Lower.Athletic")))) -> (Output_Get:bool)
  ]}
]
variables [AthleticBool {type=boolean}]
